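# Conversation database for the BaskIsle NPC, who introduces herself as Velma Dinkley.
# The entry point leads to one speech node that has a default line plus contextual lines for first-ever conversations, per-match conversation counts and specific player outfits.
# Curve rows prioritize first-interaction and outfit-reaction lines, while count-based lines have no curve.
# The speech node then links to shared NPC quest, basic service and Season 37 BaskIsle voice-fork entry points.
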ConversationDatabase /NPCLibrary/NPCs/BaskIsle/NPCLibrary_Conversation_BaskIsle
entry_points [{tag=Conversation.Entry.BaskIsle, destinations [#1]}]
nodes [
 ConversationEntryPointNode_7 {
  id=#1
  class=ConversationEntryPointNode
  outputs [#2]
  properties {entryTag {tagName=Conversation.Entry.BaskIsle}}
 }
 Speech_C_4 {
  id=#2
  class=Speech_C
  outputs [#3, #4, #5]
  properties {
   nodeName=""
   generalConfig {
    defaultMessage=Hmm.. d'you think I'm the only one who's got it together?
    contextualMessages [
     {
      message="Velma Dinkley, at your service!"
      contextRequirements [{requirement {class=FortControllerRequirement_IsFirstEverConversationWithNPC}}]
      priority {curve {rowName=Tandem.Speech.Priority.FirstEverInteraction}}
     }
     {
      message=Jinkies... this Island is like nothing I've seen before...
      contextRequirements [{requirement {
         class=FortControllerRequirement_CheckNPCMatchConversationCount
         properties {nPCUniqueIdTag {tagName=AISpawnerData.Type.Tandem.BaskIsle}, requiredConversationCountConfig {requirement=GreaterThan}}
        }}]
      priority {curve {curveTable=null, rowName=None}}
     }
     {
      message=Another mystery? I guess I don't know what I expected.
      contextRequirements [{requirement {
         class=FortControllerRequirement_CheckNPCMatchConversationCount
         properties {nPCUniqueIdTag {tagName=AISpawnerData.Type.Tandem.BaskIsle}, requiredConversationCountConfig {requirement=GreaterThan}}
        }}]
      priority {curve {curveTable=null, rowName=None}}
     }
     {
      message=Hmm.. d'you think I'm the only one who's got it together?
      contextRequirements [{requirement {
         class=FortControllerRequirement_CheckNPCMatchConversationCount
         properties {nPCUniqueIdTag {tagName=AISpawnerData.Type.Tandem.BaskIsle}, requiredConversationCountConfig {requirement=GreaterThan}}
        }}]
      priority {curve {curveTable=null, rowName=None}}
     }
     {
      message=Splitting the gang up ALWAYS ends badly.
      contextRequirements [{requirement {class=FortControllerRequirement_HasCID, properties {allowedCIds [/BRCosmetics/Athena/Items/Cosmetics/Characters/Character_OboeThorn]}}}]
      priority {curve {rowName=Tandem.Speech.Priority.OutfitReaction}}
     }
     {
      message="There's always a logical explanation. "
      contextRequirements [{requirement {
         class=FortControllerRequirement_HasCID
         properties {allowedCIds [
           /BRCosmetics/Athena/Items/Cosmetics/Characters/Character_FluteLamp
           /BRCosmetics/Athena/Items/Cosmetics/Characters/Character_LycheeNickel
           /BRCosmetics/Athena/Items/Cosmetics/Characters/Character_ArcanaAgate
          ]}
        }}]
      priority {curve {rowName=Tandem.Speech.Priority.OutfitReaction}}
     }
    ]
   }
  }
 }
 ConversationLinkNode_2 {
  id=#5
  class=ConversationLinkNode
  properties {remoteEntryTag {tagName=Conversation.Entry.Season37.VoiceFork.BaskIsle}}
 }
 ConversationLinkNode_0 {id=#4, class=ConversationLinkNode, properties {remoteEntryTag {tagName=Conversation.Entry.BasicServices}}}
 ConversationLinkNode_1 {id=#3, class=ConversationLinkNode, properties {remoteEntryTag {tagName=Conversation.Entry.NPCLibrary.Quests}}}
]
